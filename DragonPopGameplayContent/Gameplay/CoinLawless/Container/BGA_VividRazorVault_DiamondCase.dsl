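; Searchable diamond display case container from the Vivid Razor vault set, which shatters its glass when looted.
; BeginPlay creates a dynamic material and caches its emissive color multiplier, which EndPlay writes back.
; On loot it deactivates the diamond Niagara effect, then splits by authority.
; The server sends rumble and camera shake to the searching player and fires a shatter GameplayCue, while remotes hide the glass and black out the emissive.
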(event EventBeginPlay
  (bind _staticmeshcomponent (Variables|Debug|GetStaticMeshComponent))
  (|CallParentFunction)
  (bind _returnvalue (Rendering|Material|CreateDynamicMaterialInstance _staticmeshcomponent 0 "/DragonPopGameplayContent/Assets/Props/CoinContainer/Materials/MI_VividRazor_CoinCase_A_A.MI_VividRazor_CoinCase_A_A" "None" :!class /Script/Engine.PrimitiveComponent))
  (Rendering|Material|SetMaterial _staticmeshcomponent 0 _returnvalue)
  (bind _returnvalue_2 (Rendering|Material|GetVectorParameterValue _returnvalue "Emissive_Color_Multiplier" :!class /Script/Engine.MaterialInstanceDynamic))
  (Variables|Default|SetEmissiveMultiplier _returnvalue_2))

(event Collision|EventActorBeginOverlap
  (params (OtherActor :type Actor%20Object%20Reference))
  (|CallParentFunction OtherActor) ; This node is disabled and will not be called.\nDrag off pins to build functionality.
)

(event EventDestroyed
  (|CallParentFunction))

(event EventEndPlay
  (params (EndPlayReason :type EEndPlayReason%20Enum :default "Destroyed"))
  (|CallParentFunction EndPlayReason)
  (Rendering|Material|SetVectorParameterValue (Class|BTechArtGlobalLibrary|GetDynamicMaterialInstance (Variables|Debug|GetStaticMeshComponent)) "Emissive_Color_Multiplier" (Variables|Default|GetEmissiveMultiplier) :!class /Script/Engine.MaterialInstanceDynamic))

(event EventOnLoot
  (|CallParentFunction)
  (Components|Activation|Deactivate (Variables|Default|GetNS_Pickups_Diamond))
  (Networking|SwitchHasAuthority
    (:Authority
      (bind _returnvalue (Container|GetSearchingPawn))
      (Game|Feedback|ClientPlayForceFeedback (Fortnite|GetFortPlayerControllerfromActor _returnvalue) "/Game/Weapons/ForceFeedback/FF_Generic_Explosion_Small_Near.FF_Generic_Explosion_Small_Near" "None" false false false) ; Rumble
      (Game|Feedback|ClientStartCameraShake :self (Fortnite|GetFortPlayerControllerfromActor _returnvalue) :Shake "/Game/Effects/Fort_Effects/BluePrints/CameraShakes/B_Small_Explosion_CameraShake.B_Small_Explosion_CameraShake_C" :Scale 1.0 :PlaySpace "CameraLocal") ; Camera Shake
      (GameplayCue|ExecuteGameplayCueOnActor(Burst) self (Variables|Default|GetCase%20Shatter%20Cue%20Tag) (Ability|GameplayCue|MakeGameplayCueParameters :NormalizedMagnitude 1.0 :RawMagnitude 1.0 :Location (Transformation|GetWorldLocation (Variables|Default|GetItemMesh)) :Normal "0,0,0" :GameplayEffectLevel 1 :AbilityLevel 1 :TargetAttachComponent (Variables|Default|GetItemMesh) :bReplicateLocationWhenUsingMinimalRepProxy false)) ; Glass shatter burst GCN
    )
    (:Remote
      (Rendering|SetVisibility (Variables|Default|GetSM_VividRazor_CoinCase_Glass_A_A) false true)
      (Rendering|Material|SetVectorParameterValue (Class|BTechArtGlobalLibrary|GetDynamicMaterialInstance (Variables|Debug|GetStaticMeshComponent)) "Emissive_Color_Multiplier" "(R=0.000000,G=0.000000,B=0.000000,A=1.000000)" :!class /Script/Engine.MaterialInstanceDynamic))))

(fn ConstructionScript
  (|CallParentFunction))
